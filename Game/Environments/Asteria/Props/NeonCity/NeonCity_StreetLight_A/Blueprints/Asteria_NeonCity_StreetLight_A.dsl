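; Neon City street light prop A for the Asteria environment, with a spot light and a light cone mesh.
; It hooks into dynamic destruction by supplying its static mesh and light cone component lists and hiding the spot light when destruction starts.
; As an exposed asset provider, it versions itself and registers the light cone material as an indirect asset reference.
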(event DynamicDestruction|EventOnDynamicDestructionStarted
  (Development|SetHiddeninGame (Variables|Default|GetSpotLight) true false))

(fn BuildExposedAssetProviderVersion(Implementation)
  (params (Version :type Fort%20Exposed%20Asset%20Provider%20Version%20Structure%20%28by%20ref%29))
  (|CallParentFunction Version)
  (ExposedAssetProvider|AddVersion Version 1)
  (return))

(fn CollectAssetReferencestoAllow(Implementation)
  (params (Collector :type Fort%20Exposed%20Asset%20Reference%20Collector%20Structure%20%28by%20ref%29))
  (|CallParentFunction Collector)
  (ExposedAssetProvider|CollectIndirectAssetReference Collector (Utilities|ResolveSoftReference (LocalVariables|GetLightConeMat)))
  (return))

(fn GetCreativeActorBounds
  (return false "0, 0, 0"))

(fn GetCreativeActorOrigin
  (return false "0, 0, 0"))

(fn GetCreativeActorRelativeTransform
  (return false 0))

(fn GetDynamicDestructionStaticMeshComponents
  (return (Utilities|Array|MakeArray (Variables|Debug|GetStaticMeshComponent)) (Utilities|Array|MakeArray (Variables|Debug|GetStaticMeshComponent) (Variables|Default|GetLightCone)) (Utilities|Array|MakeArray (Variables|Debug|GetStaticMeshComponent))))

(fn GetGridAxisGuideAttachmentPoint
  (return true true "0, 0, 0"))

(fn ConstructionScript
  (|CallParentFunction))
