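; Parent Blueprint for the Asteria NeonCity street light prop, switching its emissive glow with time of day.
; Once the time-of-day manager is ready, it registers for phase changes, blacking out emissive tint in Day and using LightColor otherwise.
; The construction script estimates night from the global TimeOfDay_0-1 material parameter.
; Every branch gives the spotlight the same LightColor, and BeginPlay's initial check reads the update event's cached phase output.
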(event EventBeginPlay
  (bind _spotlight (Variables|Default|GetSpotLight))
  (bind _lightcolor (Variables|Default|GetLightColor))
  (Class|FortAsyncActionTimeOfDayManagerReady|GetTimeOfDayManagerReady
    (:OnReady
      (TimeOfDay|RegisterforDayPhaseChangedEvent (TimeOfDay|GetGlobalTimeOfDayManager) (delegate Custom|Time%20of%20Day%20Update) true)
      (if (Utilities|Enum|Equal(Enum) (stale Custom|Time%20of%20Day%20Update CurrentDayPhase) "Day")
        (Rendering|Material|SetVectorParameterValueonMaterials (Variables|Debug|GetStaticMeshComponent) "SFX_Emissive_Tint(Expand)" "0, 0, 0")
        (Rendering|Components|Light|SetLightColor _spotlight _lightcolor false)
        (else
          (Rendering|Material|SetVectorParameterValueonMaterials (Variables|Debug|GetStaticMeshComponent) "SFX_Emissive_Tint(Expand)" (Math|Conversions|ToVector(LinearColor) (Variables|Default|GetLightColor)))
          (Rendering|Components|Light|SetLightColor _spotlight _lightcolor false))))))

(event Collision|EventActorBeginOverlap
  (params (OtherActor :type Actor%20Object%20Reference)))

(event Custom|Time%20of%20Day%20Update
  (params (CurrentDayPhase :type EFortDayPhase%20Enum :default "Morning") (PreviousDayPhase :type EFortDayPhase%20Enum :default "Morning") (bAtCreation :type Boolean))
  (bind _spotlight (Variables|Default|GetSpotLight))
  (bind _lightcolor (Variables|Default|GetLightColor))
  (if (Utilities|Enum|Equal(Enum) CurrentDayPhase "Day")
    (Rendering|Material|SetVectorParameterValueonMaterials (Variables|Debug|GetStaticMeshComponent) "SFX_Emissive_Tint(Expand)" "0, 0, 0")
    (Rendering|Components|Light|SetLightColor _spotlight _lightcolor false)
    (else
      (Rendering|Material|SetVectorParameterValueonMaterials (Variables|Debug|GetStaticMeshComponent) "SFX_Emissive_Tint(Expand)" (Math|Conversions|ToVector(LinearColor) (Variables|Default|GetLightColor)))
      (Rendering|Components|Light|SetLightColor _spotlight _lightcolor false))))

(event EventTick
  (params (DeltaSeconds :type Float%20%28single-precision%29)))

(fn GetDynamicDestructionStaticMeshComponents
  (return (Utilities|Array|MakeArray (Variables|Debug|GetStaticMeshComponent)) (Utilities|Array|MakeArray (Variables|Debug|GetStaticMeshComponent) (Variables|Default|GetStaticMesh1)) (Utilities|Array|MakeArray (Variables|Debug|GetStaticMeshComponent))))

(fn ConstructionScript
  (bind _spotlight (Variables|Default|GetSpotLight))
  (bind _lightcolor (Variables|Default|GetLightColor))
  (|CallParentFunction)
  (bind _returnvalue (Rendering|Material|GetScalarParameterValue "/Game/Packages/Fortress_SharedMaterials/GlobalMaterialParameters/FortniteMaterialParameters.FortniteMaterialParameters" "TimeOfDay_0-1" :!class /Script/Engine.KismetMaterialLibrary))
  (if (or (> _returnvalue 0.8) (< _returnvalue 0.2))
    (Rendering|Material|SetVectorParameterValueonMaterials (Variables|Debug|GetStaticMeshComponent) "SFX_Emissive_Tint(Expand)" (Math|Conversions|ToVector(LinearColor) (Variables|Default|GetLightColor)))
    (Rendering|Components|Light|SetLightColor _spotlight _lightcolor false)
    (else
      (Rendering|Material|SetVectorParameterValueonMaterials (Variables|Debug|GetStaticMeshComponent) "SFX_Emissive_Tint(Expand)" "0, 0, 0")
      (Rendering|Components|Light|SetLightColor _spotlight _lightcolor false))))
